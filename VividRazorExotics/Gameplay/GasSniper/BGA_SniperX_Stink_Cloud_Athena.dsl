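; Gas cloud actor for the stink sniper that plays a burst sound, fades in looping audio and post-process, and fires a gameplay event to itself after an initial delay.
; A timer of duration plus delay starts fade-out and destruction.
; The fade timeline's completion registers or unregisters an AI danger zone sized from the danger sphere.
; An audio gameplay volume loops hit-notify sounds at the tick interval for the locally viewed instigator inside it.
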(event EventBeginPlay
  (Utilities|FlowControl|Sequence
    (:then_0
      (Networking|CanExecuteCosmeticEvents
        (:True
          (Audio|SpawnSoundatLocation (Variables|Default|GetStinkBulletBurstSound) (Transformation|GetActorLocation)))))
    (:then_1
      ;[1 1 1] Begin fading VFX and audio in and set timer to kill the actor after the specified duration. Since there is a delay before the damage begins and we want the duration to be the actual time that damage is being dealt, we want to kill the actor after the duration + damage delay
      (CallFunction|FadeIn)
      (Utilities|Time|SetTimerbyEvent (EventDispatchers|CreateEvent :!event Custom|HideAndKill) (+ (ScalableFloat|GetValueatLevel (Variables|Default|GetRow_CloudDuration)) (ScalableFloat|GetScalableFloatMultiplier (Variables|Default|GetRow_InitialDelay))))
      ;[1 1 1] Send event to trigger ability after delay
      (Utilities|FlowControl|Delay (ScalableFloat|GetScalableFloatMultiplier (Variables|Default|GetRow_InitialDelay)))
      (Ability|SendGameplayEventtoActor self "(TagName=\"Event.Athena.Bow.Stink.SpawnCloud\")"))))

(event Custom|HideAndKill
  ;[1 1 1] Trigger VFX and audio fade out and prepare actor to be destroyed
  (CallFunction|FadeOut)
  (Actor|SetLifeSpan 0.1)
  (CallFunction|ClearAudioShape)
  (Networking|CanExecuteCosmeticEvents
    (:True
      (CallFunction|SelfDamageTimerAudio false))))

(event Custom|FadeIn
  (bind (_origin _boxextent _sphereradius) (Collision|GetComponentBounds (Variables|Default|GetDangerSphere)))
  (bind _returnvalue (Class|AthenaAIServiceZoneManager|GetAthenaAIServiceZoneManager true))
  (bind _row_initialdelay (Variables|Default|GetRow_InitialDelay))
  (bind _returnvalue_2 (ScalableFloat|GetScalableFloatMultiplier _row_initialdelay))
  (bind _self self)
  ;[1 1 1] Fade in and out for VFX, audio, and post processing
  (Variables|Default|SetIsFadingIn true)
  (Components|Timeline|SetPlayRate (Variables|BGA_SniperX_Stink_Cloud_Athena|GetFadeSmokeTL) (Utilities|Reciprocal _returnvalue_2))
  (Audio|Components|Audio|FadeIn (Variables|Default|GetLoopingAudio) _returnvalue_2 1.0 0.0 "Linear" :!class /Script/Engine.AudioComponent)
  (bind (_direction _float) (Timeline|FadeSmokeTL PlayFromStart)
    (:Update
      (Variables|/Script/Engine.PostProcessComponent|SetBlendWeight (* _float 0.7) (Variables|Default|GetPostProcessComponent)))
    (:Finished
      (Utilities|IsValid _returnvalue
        (:Is%20Valid
          (if (Variables|Default|GetIsFadingIn)
            (Class|AthenaAIServiceZoneManager|RegisterDangerZone _returnvalue _self _sphereradius)
            (else
              (Class|AthenaAIServiceZoneManager|UnregisterDangerZone _returnvalue _self))))))))

(event Custom|FadeOut
  (bind (_origin _boxextent _sphereradius) (Collision|GetComponentBounds (Variables|Default|GetDangerSphere)))
  (bind _returnvalue (Class|AthenaAIServiceZoneManager|GetAthenaAIServiceZoneManager true))
  (bind _self self)
  ;[1 1 1] Fade in and out for VFX, audio, and post processing
  (Variables|Default|SetIsFadingIn false)
  (Components|Timeline|SetPlayRate (Variables|BGA_SniperX_Stink_Cloud_Athena|GetFadeSmokeTL) (Utilities|Reciprocal (Variables|Default|GetGasFadeOutTime)))
  (Audio|Components|Audio|FadeOut (Variables|Default|GetLoopingAudio) (Variables|Default|GetGasFadeOutTime) 0.0 "Linear" :!class /Script/Engine.AudioComponent)
  (Components|Activation|Deactivate (Variables|Default|GetNS_StinkCloud))
  (Timeline|FadeSmokeTL ReverseFromEnd))

(event OnToggledOn(AudioGameplayVolume)
  (bind _returnvalue (Fortnite|GetFirstLocalFortPlayerController true true))
  (bind _returnvalue_2 (Pawn|GetControlledPawn _returnvalue))
  (bind _instigator (Variables|Actor|GetInstigator))
  ; Play damage audio if instigator is self
  (Networking|CanExecuteCosmeticEvents
    (:True
      (Utilities|IsValid _returnvalue
        (:Is%20Valid
          (Utilities|IsValid _returnvalue_2
            (:Is%20Valid
              (if (== _returnvalue_2 _instigator)
                (if (Pawn|IsLocallyViewed _instigator)
                  (CallFunction|SelfDamageTimerAudio true))))))))))

(event OnToggledOff(AudioGameplayVolume)
  ; Play damage audio if instigator is self
  (Networking|CanExecuteCosmeticEvents
    (:True
      (CallFunction|SelfDamageTimerAudio false))))

(event Custom|SelfDamageTimerAudio
  (params (bActivate :type Boolean))
  (bind _selfdamageaudiotimer (Variables|Default|GetSelfDamageAudioTimer))
  (bind _returnvalue (Utilities|Time|IsValidTimerHandle _selfdamageaudiotimer))
  ; Play damage audio if instigator is self
  (if bActivate
    (if (not _returnvalue)
      (bind _returnvalue_2 (Utilities|Time|SetTimerbyEvent (delegate Custom|PlaySelfDamageAudio) (ScalableFloat|GetValueatLevel (Variables|Default|GetRow_TickInterval)) true))
      (Variables|Default|SetSelfDamageAudioTimer _returnvalue_2))
    (elif _returnvalue
      (Utilities|Time|ClearandInvalidateTimerbyHandle (Variables|Default|GetSelfDamageAudioTimer)))))

(event Custom|PlaySelfDamageAudio
  ; Play damage audio if instigator is self
  (bind _instigator (Variables|Actor|GetInstigator)
    (:then
      (SoundLibrary|PlaySoundSimple (WorldSubsystems|GetSoundLibrarySubsystem) _instigator (Utilities|Struct|MakeSimpleSoundSettings "(TagName=\"SoundLibrary.HitNotify.Receive.Body\")") :!class /Script/SoundLibrary.SoundLibrarySubsystem))
    (:else)))

(event Custom|ClearAudioShape
  ; Play damage audio if instigator is self
  (bind _audiogameplayvolume (Variables|Default|GetAudioGameplayVolume)
    (:then
      (Components|DestroyComponent _audiogameplayvolume))
    (:else
      _))
  (bind _audioshapesphere (Variables|Default|GetAudioShapeSphere)
    (:then
      (Components|DestroyComponent _audioshapesphere))
    (:else)))

(fn ConstructionScript
  (|CallParentFunction))
