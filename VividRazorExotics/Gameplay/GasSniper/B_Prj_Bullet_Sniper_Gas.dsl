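; Gas sniper bullet projectile Blueprint that leaves a stink cloud where it stops.
; On projectile stop, it breaks the hit result for the impact location and spawns the SniperX stink cloud building gameplay actor there.
; The cloud's instigator is taken from the projectile's owning weapon.
; BeginPlay, overlap and tick events carry no logic.
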(event EventBeginPlay
  (|CallParentFunction) ; This node is disabled and will not be called.\nDrag off pins to build functionality.
)

(event OnProjectileStop(ProjectileMovementComponent)
  (params (ImpactResult :type Hit%20Result%20Structure%20%28by%20ref%29))
  (bind (_bblockinghit _binitialoverlap _time _distance _location) (Collision|BreakHitResult ImpactResult))
  (bind _returnvalue (CallFunction|GetOwnerWeapon))
  (Building|SpawnBuildingGameplayActor "/VividRazorExotics/Gameplay/GasSniper/BGA_SniperX_Stink_Cloud_Athena.BGA_SniperX_Stink_Cloud_Athena_C" (Math|Conversions|ToTransform(Vector) _location) (Game|GetInstigator _returnvalue)))

(event Collision|EventActorBeginOverlap
  (params (OtherActor :type Actor%20Object%20Reference)))

(event EventTick
  (params (DeltaSeconds :type Float%20%28single-precision%29)))

(fn ConstructionScript
  (|CallParentFunction))
